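; Burst gameplay cue for the Athena Realm Throw environment impact, placing and orienting its particle effect and decal.
; After the parent burst runs, it sets the first particle component's HitPlayer parameter to false and line-traces along the cue's normal through the impact location.
; On a hit, the decal snaps to the impact point and surface normal.
; The particle rotates to the cue normal reflected off that surface, then gets a 180-degree local flip.
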(fn OnBurst(EitherVFX-preferthisversion)
  (params (MyTarget :type Actor%20Object%20Reference) (Parameters :type Gameplay%20Cue%20Parameters%20Structure%20%28by%20ref%29) (ParticleComponents :type Array%20of%20FXSystem%20Component%20Object%20References) (AudioComponents :type Array%20of%20Audio%20Component%20Object%20References) (BurstCameraShakeInstance :type Camera%20Shake%20Base%20Object%20Reference) (BurstDecalInstance :type Decal%20Actor%20Object%20Reference))
  (bind (_normalizedmagnitude _rawmagnitude _effectcontext _matchedtagname _originaltag _aggregatedsourcetags _aggregatedtargettags _location _normal) (Ability|GameplayCue|BreakGameplayCueParameters Parameters))
  (bind _returnvalue (Math|Conversions|RotationFromXVector _normal))
  (bind _returnvalue_2 (Math|Vector|GetRightVector _returnvalue))
  (bind _output (Utilities|Array|Get(acopy) ParticleComponents))
  (bind _decal (Variables|/Script/Engine.DecalActor|GetDecal BurstDecalInstance))
  (|CallParentFunction MyTarget Parameters ParticleComponents AudioComponents BurstCameraShakeInstance BurstDecalInstance)
  (Effects|Components|ParticleSystem|SetBoolParameter _output "HitPlayer" false)
  ;[1 1 1] Ground Trace
  (bind (_outhit _returnvalue_3) (Collision|LineTraceByChannel 0 (- _location (* _normal 2e+01)) (+ _location (* _normal 6e+02)) "TraceTypeQuery10"))
  (if _returnvalue_3
    ;[1 1 1] Set the Decal Location and Rotation
    (Transformation|SetWorldLocation _decal (.ImpactPoint (Collision|BreakHitResult _outhit)) false false)
    (Transformation|SetWorldRotation _output (Math|Rotator|MakeRotfromXY (Math|Vector|GetReflectionVector _normal (.ImpactNormal (Collision|BreakHitResult _outhit))) _returnvalue_2) false false)
    (Transformation|AddLocalRotation _output "0.000000,0.000000,180.000000" false false)
    (Transformation|SetWorldRotation _decal (Math|Rotator|MakeRotfromXY (.ImpactNormal (Collision|BreakHitResult _outhit)) _returnvalue_2) false false)
    (Transformation|AddLocalRotation _decal "0, 0, 0" false false)
    (return)))
