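; Burst gameplay cue that plays impact effects and sound when a Realm throw projectile hits a player.
; On burst it runs the parent logic, then sets the HitPlayer parameter on the first particle component to false, despite the player-impact name.
; It spawns the player-hit sound at the cue location and sets its IsLocal parameter by whether the instigator's owning player pawn is locally viewed.
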(fn OnBurst(EitherVFX-preferthisversion)
  (params (MyTarget :type Actor%20Object%20Reference) (Parameters :type Gameplay%20Cue%20Parameters%20Structure%20%28by%20ref%29) (ParticleComponents :type Array%20of%20FXSystem%20Component%20Object%20References) (AudioComponents :type Array%20of%20Audio%20Component%20Object%20References) (BurstCameraShakeInstance :type Camera%20Shake%20Base%20Object%20Reference) (BurstDecalInstance :type Decal%20Actor%20Object%20Reference))
  (bind (_normalizedmagnitude _rawmagnitude _effectcontext _matchedtagname _originaltag _aggregatedsourcetags _aggregatedtargettags _location _normal _instigator) (Ability|GameplayCue|BreakGameplayCueParameters Parameters))
  (|CallParentFunction MyTarget Parameters ParticleComponents AudioComponents BurstCameraShakeInstance BurstDecalInstance)
  (Effects|Components|ParticleSystem|SetBoolParameter (Utilities|Array|Get(acopy) ParticleComponents) "HitPlayer" false)
  (bind _asplayer_pawn_athena (Utilities|Casting|CastToPlayerPawn_Athena (Actor|GetOwner _instigator)))
  (bind _returnvalue (Audio|SpawnSoundatLocation _instigator "/Galileo/Items/Lobster/Assets/Sounds/Realm/Lobster_Realm_Throw_Prj_Hit_Player_Cue.Lobster_Realm_Throw_Prj_Hit_Player_Cue" _location))
  (Audio|Parameter|SetBooleanParameter _returnvalue "IsLocal" (Pawn|IsLocallyViewed _asplayer_pawn_athena) :!class /Script/AudioExtensions.AudioParameterControllerInterface))
