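; Burst gameplay cue for the impact of Galileo Lobster Attack 4, placing its ground decal and impact particle effect.
; After the parent OnBurst, a short vertical line trace around the spawned decal finds the ground.
; On a hit, decal and particle move there, face the player's root socket with rotation offsets, and the particle takes hot/cool blade colors from the current weapon.
; On a miss, only the particle is hidden, although a comment mentions the decal too.
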(fn OnBurst(EitherVFX-preferthisversion)
  (params (MyTarget :type Actor%20Object%20Reference) (Parameters :type Gameplay%20Cue%20Parameters%20Structure%20%28by%20ref%29) (ParticleComponents :type Array%20of%20FXSystem%20Component%20Object%20References) (AudioComponents :type Array%20of%20Audio%20Component%20Object%20References) (BurstCameraShakeInstance :type Camera%20Shake%20Base%20Object%20Reference) (BurstDecalInstance :type Decal%20Actor%20Object%20Reference))
  (bind _asplayer_pawn_athena (Utilities|Casting|CastToPlayerPawn_Athena MyTarget))
  (bind _mesh (Variables|/Script/Engine.Character|GetMesh _asplayer_pawn_athena))
  (bind _returnvalue (Transformation|GetSocketLocation _mesh "root"))
  (bind _currentweapon (Variables|/Script/FortniteGame.FortPawn|GetCurrentWeapon _asplayer_pawn_athena))
  (bind _asb_athena_galileo_lobster_parent (Utilities|Casting|CastToB_Athena_Galileo_Lobster_Parent _currentweapon))
  (bind _output (Utilities|Array|Get(acopy) ParticleComponents))
  (bind _decal (Variables|/Script/Engine.DecalActor|GetDecal BurstDecalInstance))
  (|CallParentFunction MyTarget Parameters ParticleComponents AudioComponents BurstCameraShakeInstance BurstDecalInstance)
  ;[1 1 1] Ground Trace
  (bind (_outhit _returnvalue_2) (Collision|LineTraceByChannel 0 (+ (Transformation|GetWorldLocation _decal) "0, 0,100.000000") (+ (Transformation|GetWorldLocation _decal) "0, 0,-100.000000") "TraceTypeQuery10"))
  (if _returnvalue_2
    ;[1 1 1] Set the Decal Location and Rotation
    (Transformation|SetWorldLocationAndRotation _decal (.ImpactPoint (Collision|BreakHitResult _outhit)) (Math|Rotator|FindLookatRotation (.ImpactPoint (Collision|BreakHitResult _outhit)) _returnvalue) false false)
    (Transformation|AddLocalRotation _decal "-90.000000,90.000000, 0" false false)
    ;[1 1 1] Set the Impact FX Location, Rotation, and Color
    (Transformation|SetWorldLocationAndRotation _output (.ImpactPoint (Collision|BreakHitResult _outhit)) (Math|Rotator|FindLookatRotation (.ImpactPoint (Collision|BreakHitResult _outhit)) _returnvalue) false false)
    (Transformation|AddLocalRotation _output "0.000000,180.000000,90.000000" false false)
    (Effects|Components|ParticleSystem|SetColorParameter _output "Color_Blade_A" (Variables|/Galileo/Items/Lobster/Actors/B_Athena_Galileo_Lobster_Parent.B_Athena_Galileo_Lobster_Parent_C|GetBlade%20Color%20Hot _asb_athena_galileo_lobster_parent))
    (Effects|Components|ParticleSystem|SetColorParameter _output "Color_Blade_B" (Variables|/Galileo/Items/Lobster/Actors/B_Athena_Galileo_Lobster_Parent.B_Athena_Galileo_Lobster_Parent_C|GetBlade%20Color%20Cool _asb_athena_galileo_lobster_parent))
    (Effects|Components|ParticleSystem|SetBoolParameter _output "HitPlayer" false)
    (return)
    (else
      (Rendering|SetVisibility _output false false) ; Hide the impact decal and fx if the ground trace fails
      (return))))
