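; Gameplay cue burst for the Galileo Lobster block break, presenting the broken-block effect and sound.
; OnBurst runs the parent, then plays a tagged Sound Library sound on the source object actor, positioned 200 units above it.
; It sets IsBroken and two blade colors on the first spawned particle component.
; The cue's Location and Normal vectors are repurposed to carry those colors.
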(fn OnBurst(EitherVFX-preferthisversion)
  (params (MyTarget :type Actor%20Object%20Reference) (Parameters :type Gameplay%20Cue%20Parameters%20Structure%20%28by%20ref%29) (ParticleComponents :type Array%20of%20FXSystem%20Component%20Object%20References) (AudioComponents :type Array%20of%20Audio%20Component%20Object%20References) (BurstCameraShakeInstance :type Camera%20Shake%20Base%20Object%20Reference) (BurstDecalInstance :type Decal%20Actor%20Object%20Reference))
  (bind (_normalizedmagnitude _rawmagnitude _effectcontext _matchedtagname _originaltag _aggregatedsourcetags _aggregatedtargettags _location _normal _instigator _effectcauser _sourceobject) (Ability|GameplayCue|BreakGameplayCueParameters Parameters))
  (bind _output (Utilities|Array|Get(acopy) ParticleComponents))
  (|CallParentFunction MyTarget Parameters ParticleComponents AudioComponents BurstCameraShakeInstance BurstDecalInstance)
  (Default|PlaySoundLibrary _sourceobject)
  (Effects|Components|ParticleSystem|SetColorParameter _output "Color_Blade_A" (Math|Conversions|ToLinearColor(Vector) _location))
  (Effects|Components|ParticleSystem|SetColorParameter _output "Color_Blade_B" (Math|Conversions|ToLinearColor(Vector) _normal))
  (Effects|Components|ParticleSystem|SetBoolParameter _output "IsBroken" true)
  (return))

(fn PlaySoundLibrary
  (params (Actor :type Object%20Reference))
  (SoundLibrary|PlaySoundSimple (WorldSubsystems|GetSoundLibrarySubsystem 0) (Utilities|Casting|CastToActor Actor) (Utilities|Struct|MakeSimpleSoundSettings (Variables|Default|GetSoundLibraryTag) false true true false 1.0 1.0 0.0 0 0 "None" 0 "KeepRelativeOffset" (+ (Transformation|GetActorLocation (Utilities|Casting|CastToActor Actor)) "0,0,200.000000")) :!class /Script/SoundLibrary.SoundLibrarySubsystem))
